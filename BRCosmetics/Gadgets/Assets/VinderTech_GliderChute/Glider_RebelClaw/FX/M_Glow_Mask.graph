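# Masked, unlit glow material for Niagara sprites in the Rebel Claw glider FX.
# An exponential radial gradient with adjustable density, offset by a shading parameter, tints particle color and passes through shared emissive adjustments.
# The mask multiplies particle alpha by the gradient, then applies depth fade and view-angle fade.
# A very low clip threshold keeps most of the faint gradient visible despite masked blending.
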Material /BRCosmetics/Gadgets/Assets/VinderTech_GliderChute/Glider_RebelClaw/FX/M_Glow_Mask
graph {
 properties {
  bUsedWithNiagaraSprites=true
  blendMode=BLEND_Masked
  opacityMaskClipValue=0.01
  pixelDepthOffsetMode=PDOM_Legacy
  shadingModel=MSM_Unlit
  usedShadingModels=Unlit
 }
 nodes [
  n1 = ParticleColor()
  n2 = ScalarParameter() {defaultValue=0.01, parameterName=Shading}
  n3 = MaterialFunctionCall(Density=ScalarParameter() {defaultValue=0.5, parameterName=Density}) {materialFunction=&RadialGradientExponential}
  MaterialFunctionCall(Emissive=Multiply(A=@n1.RGB, B=Add(A=@n2, B=@n3.RadialGradientExponential) {constB=0.5})) {materialFunction=&MF_FX_EmissiveAdjustments}
  Multiply(
   A=DepthFade(
    Opacity=Multiply(A=@n1.A, B=@n3.RadialGradientExponential)
    FadeDistance=ScalarParameter() {defaultValue=4, parameterName=DepthFade_Distance}
   ) {fadeDistanceDefault=8}
   B=MaterialFunctionCall("FadeSensitivity (optional)"=ScalarParameter() {defaultValue=2, parameterName=ViewAngleFade_Senstivity}) {materialFunction=&FadeBasedOnViewAngle}.Result
  )
 ]
}
