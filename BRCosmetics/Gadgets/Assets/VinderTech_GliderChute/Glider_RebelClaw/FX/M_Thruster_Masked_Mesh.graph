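# Unlit, dithered masked material for the thruster exhaust effect on the Rebel Claw glider, usable on Niagara mesh and sprite particles.
# Two panning fire masks combine with a V-gradient falloff to blend a smoke color into a thruster-to-core gradient, tinted by particle color and passed through MF_FX_EmissiveAdjustments.
# A mask term multiplies V gradient and particle alpha by that blend plus squared, panning swirl noise.
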Material /BRCosmetics/Gadgets/Assets/VinderTech_GliderChute/Glider_RebelClaw/FX/M_Thruster_Masked_Mesh
graph {
 properties {
  bCastRayTracedShadows=false
  bUsedWithNiagaraMeshParticles=true
  bUsedWithNiagaraSprites=true
  blendMode=BLEND_Masked
  ditherOpacityMask=true
  opacityMaskClipValue=0.1
  pixelDepthOffsetMode=PDOM_Legacy
  shadingModel=MSM_Unlit
  usedShadingModels=Unlit
 }
 nodes [
  n1 = VectorParameter() {parameterName=Color_SmokeBlend, defaultValue=<rgba 1 0.1667 0 1>}
  n2 = ComponentMask(None=TextureCoordinate()) {g=true}
  n3 = LinearInterpolate(
   A=VectorParameter() {parameterName=Color_Thruster, defaultValue=<rgba 0 0 1 1>}.RGB
   B=VectorParameter() {parameterName=Color_ThrusterCore, defaultValue=<rgba 0 2.6667 8 1>}.RGB
   Alpha=Subtract(A=@n2) {constB=0.5}
  )
  n4 = TextureSample(UVs=Panner(Coordinate=TextureCoordinate() {uTiling=0.25, vTiling=1.2}) {speedX=0.25}) {texture=/Game/Effects/Fort_Effects/Textures/Fire/T_Fire_Mask}
  n5 = TextureSample(UVs=Panner(Coordinate=TextureCoordinate() {uTiling=0.15, unMirrorU=true, unMirrorV=true, vTiling=0.9}) {speedX=-0.1}) {texture=&T_Fire_Mask}
  n6 = LinearInterpolate(A=Saturate(None=Add(A=@n4.RGB, B=@n5.RGB)), B=Power(Base=@n2) {constExponent=8}) {constAlpha=0.75}
  n7 = ParticleColor()
  MaterialFunctionCall(Emissive=Multiply(A=LinearInterpolate(A=@n1.RGB, B=@n3, Alpha=@n6), B=@n7.RGB)) {materialFunction=&MF_FX_EmissiveAdjustments}
  n8 = TextureSample(UVs=Panner(Coordinate=TextureCoordinate()) {speedX=0.3, speedY=3}) {samplerType=SAMPLERTYPE_LinearGrayscale, texture=/Game/Global/Textures/Default/Noise/T_SwirlNoise_N_height}
  Multiply(A=Multiply(A=@n2, B=@n7.A), B=Add(A=@n6, B=Multiply(A=@n8.RGB, B=@n8.RGB)))
 ]
}
