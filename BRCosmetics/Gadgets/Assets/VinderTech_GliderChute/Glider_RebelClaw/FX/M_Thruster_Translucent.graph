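# Unlit translucent Niagara sprite material for thruster effects on the Rebel Claw glider.
# A radial gradient blends thruster and core colors, which mix with a smoke color along the V texture coordinate.
# The result is tinted toward particle color and passed through the shared FX emissive adjustment function.
# A second radial gradient, scaled by particle alpha, is softened by depth fade and view-angle fade to form a translucency mask.
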Material /BRCosmetics/Gadgets/Assets/VinderTech_GliderChute/Glider_RebelClaw/FX/M_Thruster_Translucent
graph {
 properties {bUsedWithNiagaraSprites=true, blendMode=BLEND_Translucent, pixelDepthOffsetMode=PDOM_Legacy, shadingModel=MSM_Unlit, usedShadingModels=Unlit}
 nodes [
  n1 = VectorParameter() {parameterName=Color_SmokeBlend, defaultValue=<rgba 5 0.8333 0 1>}
  n2 = LinearInterpolate(
   A=VectorParameter() {parameterName=Color_Thruster, defaultValue=<rgba 0 0 5 1>}.RGB
   B=VectorParameter() {parameterName=Color_ThrusterCore, defaultValue=<rgba 0 6.6667 20 1>}.RGB
   Alpha=MaterialFunctionCall(Radius=Constant() {r=0.333}, Density=Constant() {r=0.5}) {materialFunction=&RadialGradientExponential}.RadialGradientExponential
  )
  n3 = Divide(A=ComponentMask(None=TextureCoordinate()) {g=true}) {constB=0.333}
  n4 = ParticleColor()
  MaterialFunctionCall(Emissive=LinearInterpolate(A=LinearInterpolate(A=@n1.RGB, B=@n2, Alpha=@n3), B=@n4.RGB) {constAlpha=0.333}) {materialFunction=&MF_FX_EmissiveAdjustments}
  n5 = MaterialFunctionCall(Density=ScalarParameter() {defaultValue=0.5, parameterName=Density}) {materialFunction=&RadialGradientExponential}
  Multiply(
   A=DepthFade(
    Opacity=Multiply(A=@n4.A, B=@n5.RadialGradientExponential)
    FadeDistance=ScalarParameter() {defaultValue=8, parameterName=DepthFade_Distance}
   ) {fadeDistanceDefault=8}
   B=MaterialFunctionCall("FadeSensitivity (optional)"=ScalarParameter() {defaultValue=4, parameterName=ViewAngleFade_Senstivity}) {materialFunction=&FadeBasedOnViewAngle}.Result
  )
 ]
}
